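# Physics asset for the face accessory part of the male medium SpringBreak cosmetic body, in a Frontend variant, driving dangling bone chains.
# Capsule bodies cover branching dyn_ chains plus head, neck, clavicles and upper spine.
# Most chains root at head, with F and I branching off the G chain; B hangs from E_1, which has no parent constraint.
# Swing and twist limits are tuned per chain, with twist locked on C and annoyingChain.
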PhysicsAsset /BRCosmetics/Characters/Player/Male/Medium/Bodies/M_MED_SpringBreak/Meshes/Parts/M_MED_SpringBreak_FaceAcc_Physics_Frontend
bodies {
 clavicle_l [{shapeType=Capsule, length=5, radius=10, shapeName=clavicle_l_capsule, center=<10.0648 -1.0199 -2.2332>, rotation=<rot 78.1962 0 0>}]
 clavicle_r [{shapeType=Capsule, length=5, radius=10, shapeName=clavicle_l_capsule, center=<-10.0648 1.0199 2.2332>, rotation=<rot 78.1962 0 0>}]
 dyn_annoyingChain_1 [{shapeType=Capsule, length=4, radius=2.5, shapeName=dyn_annoyingChain_1_capsule, center=<4 0 0>, rotation=<rot 90 0 0>}]
 dyn_annoyingChain_2 [{shapeType=Capsule, length=5, radius=2.5, shapeName=dyn_annoyingChain_2_capsule, center=<4 0 0>, rotation=<rot 90 0 0>}]
 dyn_annoyingChain_3 [{shapeType=Capsule, length=4, radius=2.5, shapeName=dyn_annoyingChain_3_capsule, center=<4 0 0>, rotation=<rot 90 0 0>}]
 dyn_chain_B_1 [{shapeType=Capsule, length=6, radius=1.5, shapeName=dyn_chain_B_1_capsule, center=<3 0 0>, rotation=<rot 90 0 0>}]
 dyn_chain_B_2 [{shapeType=Capsule, length=6, radius=1.5, shapeName=dyn_chain_B_2_capsule, center=<3 0 0>, rotation=<rot 90 0 0>}]
 dyn_chain_B_3 [{shapeType=Capsule, length=6, radius=1.5, shapeName=dyn_chain_B_3_capsule, center=<3 0 0>, rotation=<rot 90 0 0>}]
 dyn_chain_B_4 [{shapeType=Capsule, length=6, radius=1.5, shapeName=dyn_chain_B_4_capsule, center=<3 0 0>, rotation=<rot 90 0 0>}]
 dyn_chain_B_5 [{shapeType=Capsule, length=6, radius=1.5, shapeName=dyn_chain_B_5_capsule, center=<3 0 0>, rotation=<rot 90 0 0>}]
 dyn_chain_B_6 [{shapeType=Capsule, length=6, radius=1.5, shapeName=dyn_chain_B_6_capsule, center=<3 0 0>, rotation=<rot 90 0 0>}]
 dyn_chain_C_1 [{shapeType=Capsule, length=4, radius=2, shapeName=dyn_chain_C_1_capsule, center=<4 0 0>, rotation=<rot 90 0 0>}]
 dyn_chain_C_2 [{shapeType=Capsule, length=4, radius=2, shapeName=dyn_chain_C_2_capsule, center=<4 0 0>, rotation=<rot 90 0 0>}]
 dyn_chain_C_3 [{shapeType=Capsule, length=4, radius=2, shapeName=dyn_chain_C_3_capsule, center=<4 0 0>, rotation=<rot 90 0 0>}]
 dyn_chain_D_1 [{shapeType=Capsule, length=4, radius=2.5, shapeName=dyn_chain_D_1_capsule, center=<4 0 0>, rotation=<rot 90 0 0>}]
 dyn_chain_D_2 [{shapeType=Capsule, length=4, radius=2.5, shapeName=dyn_chain_D_2_capsule, center=<4 0 0>, rotation=<rot 90 0 0>}]
 dyn_chain_D_5 [{shapeType=Capsule, length=4, radius=3.5, shapeName=dyn_chain_D_5_capsule, center=<4 -0.5 0>, rotation=<rot 90 0 0>}]
 dyn_chain_D_6 [{shapeType=Capsule, length=4, radius=3.5, shapeName=dyn_chain_D_6_capsule, center=<4 -0.5 0>, rotation=<rot 90 0 0>}]
 dyn_chain_D_7 [{shapeType=Capsule, length=5, radius=2.5, shapeName=dyn_chain_D_7_capsule, center=<4 0 0>, rotation=<rot 90 0 0>}]
 dyn_chain_E_1 [{shapeType=Capsule, length=1, radius=1, shapeName=dyn_chain_E_1_capsule}]
 dyn_chain_F_1 [
  {shapeType=Capsule, length=5, radius=4, shapeName=dyn_chain_F_1_capsule, center=<6 0 0.0655>, rotation=<rot 90 0 0>}
  {shapeType=Capsule, length=7, radius=3, shapeName=dyn_chain_F_1_capsule, center=<5.8178 -1.8824 -2.6771>, rotation=<rot 90 0 0>}
 ]
 dyn_chain_F_2 [
  {shapeType=Capsule, length=4, radius=3, shapeName=dyn_chain_F_2_capsule, center=<4 0 0>, rotation=<rot 90 0 0>}
  {
   shapeType=Capsule
   length=9
   radius=3
   shapeName=dyn_chain_F_2_capsule
   center=<0.7706 -2.1938 -1.5767>
   rotation=<rot 80.0001 169.9999 179.9999>
  }
 ]
 dyn_chain_F_3 [
  {shapeType=Capsule, length=4, radius=3, shapeName=dyn_chain_F_3_capsule, center=<4 0 0>, rotation=<rot 90 0 0>}
  {shapeType=Capsule, length=5, radius=2, shapeName=dyn_chain_F_3_capsule, center=<3 -5 -0.4318>, rotation=<rot 90 0 0>}
 ]
 dyn_chain_F_4 [{shapeType=Capsule, length=4, radius=3, shapeName=dyn_chain_F_4_capsule, center=<4 0 0>, rotation=<rot 90 0 0>}]
 dyn_chain_F_5 [{shapeType=Capsule, length=4, radius=2.5, shapeName=dyn_chain_F_5_capsule, center=<4 -1.3 0>, rotation=<rot 90 0 0>}]
 dyn_chain_G_1 [{shapeType=Capsule, length=4, radius=3.5, shapeName=dyn_chain_G_1_capsule, center=<4 0 0>, rotation=<rot 90 0 0>}]
 dyn_chain_G_2 [{shapeType=Capsule, length=6, radius=4.5, shapeName=dyn_chain_G_2_capsule, center=<4 0 0>, rotation=<rot 90 0 0>}]
 dyn_chain_G_3 [{shapeType=Capsule, length=4, radius=3.7, shapeName=dyn_chain_G_3_capsule, center=<4 0 0>, rotation=<rot 90 0 0>}]
 dyn_chain_G_4 [{shapeType=Capsule, length=4, radius=3.9, shapeName=dyn_chain_G_4_capsule, center=<4 0 0>, rotation=<rot 90 0 0>}]
 dyn_chain_G_5 [{shapeType=Capsule, length=4, radius=3.9, shapeName=dyn_chain_G_5_capsule, center=<4 0 0>, rotation=<rot 90 0 0>}]
 dyn_chain_G_base [{shapeType=Capsule, length=3, radius=3, shapeName=dyn_chain_G_base_capsule, center=<4 0 0>, rotation=<rot 90 0 0>}]
 dyn_chain_I_1 [{shapeType=Capsule, length=4, radius=3.5, shapeName=dyn_chain_I_1_capsule, center=<4 0 0>, rotation=<rot 90 0 0>}]
 dyn_chain_I_2 [{shapeType=Capsule, length=4, radius=4, shapeName=dyn_chain_I_2_capsule, center=<4 0 0>, rotation=<rot 90 0 0>}]
 dyn_chain_I_3 [{shapeType=Capsule, length=4, radius=4.5, shapeName=dyn_chain_I_3_capsule, center=<4 0 0>, rotation=<rot 90 0 0>}]
 dyn_chain_I_4 [{shapeType=Capsule, length=4, radius=3, shapeName=dyn_chain_I_4_capsule, center=<4 0 0>, rotation=<rot 90 0 0>}]
 dyn_chain_I_5 [{shapeType=Capsule, length=10, radius=3, shapeName=dyn_chain_I_5_capsule, center=<4 0 1>, rotation=<rot 90 0 0>}]
 dyn_chain_I_6 [{shapeType=Capsule, length=6, radius=3, shapeName=dyn_chain_I_6_capsule, center=<4 0 0>, rotation=<rot 90 0 0>}]
 dyn_chain_I_7 [{shapeType=Capsule, length=9, radius=3.9, shapeName=dyn_chain_I_7_capsule, center=<4 0 1.3>, rotation=<rot 90 16.1312 -3.8688>}]
 dyn_chain_J_1 [{shapeType=Capsule, length=4, radius=3, shapeName=dyn_chain_J_1_capsule, center=<4 0 0>, rotation=<rot 90 0 0>}]
 dyn_chain_J_10 [{shapeType=Capsule, length=6, radius=2, shapeName=dyn_chain_J_10_capsule, center=<4 0 0>, rotation=<rot 90 0 0>}]
 dyn_chain_J_11 [{shapeType=Capsule, length=6, radius=2, shapeName=dyn_chain_J_10_capsule, center=<4 0 0>, rotation=<rot 90 0 0>}]
 dyn_chain_J_12 [{shapeType=Capsule, length=6, radius=2, shapeName=dyn_chain_J_10_capsule, center=<4 0 0>, rotation=<rot 90 0 0>}]
 dyn_chain_J_2 [{shapeType=Capsule, length=4, radius=3, shapeName=dyn_chain_J_2_capsule, center=<4 0 0>, rotation=<rot 90 0 0>}]
 dyn_chain_J_3 [{shapeType=Capsule, length=4.5, radius=2, shapeName=dyn_chain_J_3_capsule, center=<4 -1.7 0>, rotation=<rot 90 0 0>}]
 dyn_chain_J_4 [{shapeType=Capsule, length=4.5, radius=1.5, shapeName=dyn_chain_J_4_capsule, center=<4 0 0>, rotation=<rot 90 0 0>}]
 dyn_chain_J_7 [{shapeType=Capsule, length=4.5, radius=3, shapeName=dyn_chain_J_7_capsule, center=<4 0 0>, rotation=<rot 90 0 0>}]
 dyn_chain_J_8 [{shapeType=Capsule, length=6, radius=2, shapeName=dyn_chain_J_8_capsule, center=<4 0 0>, rotation=<rot 90 0 0>}]
 dyn_chain_J_9 [{shapeType=Capsule, length=6, radius=2, shapeName=dyn_chain_J_9_capsule, center=<4 0 0>, rotation=<rot 90 0 0>}]
 head [{shapeType=Capsule, length=5, radius=2, shapeName=head_capsule, center=<5 0 0>, rotation=<rot 90 0 0>}]
 neck_01 [{shapeType=Capsule, length=90, radius=11.5, shapeName=neck_01_capsule, center=<7.9047 7.643 0>, rotation=<rot 0 179.9998 -180>}]
 neck_02 [{shapeType=Capsule, length=7, radius=6, shapeName=neck_02_capsule, rotation=<rot 90 0 0>}]
 spine_05 [{shapeType=Capsule, length=9.3569, radius=10.5, shapeName=spine_05_capsule, center=<-2 0.1369 0>}]
}
constraints [
 {bone1Name=dyn_annoyingChain_1, bone2Name=head, swing1LimitDegrees=45, swing2LimitDegrees=45, twistLimitDegrees=5, twistMotion=Locked}
 {
  bone1Name=dyn_annoyingChain_2
  bone2Name=dyn_annoyingChain_1
  swing1LimitDegrees=45
  swing2LimitDegrees=45
  twistLimitDegrees=5
  twistMotion=Locked
 }
 {
  bone1Name=dyn_annoyingChain_3
  bone2Name=dyn_annoyingChain_2
  swing1LimitDegrees=45
  swing2LimitDegrees=45
  twistLimitDegrees=5
  twistMotion=Locked
 }
 {bone1Name=dyn_chain_B_1, bone2Name=dyn_chain_E_1, swing1LimitDegrees=45, swing2LimitDegrees=30, twistLimitDegrees=45}
 {bone1Name=dyn_chain_B_2, bone2Name=dyn_chain_B_1, swing1LimitDegrees=45, swing2LimitDegrees=30, twistLimitDegrees=45}
 {bone1Name=dyn_chain_B_3, bone2Name=dyn_chain_B_2, swing1LimitDegrees=45, swing2LimitDegrees=30, twistLimitDegrees=45}
 {bone1Name=dyn_chain_B_4, bone2Name=dyn_chain_B_3, swing1LimitDegrees=45, swing2LimitDegrees=30, twistLimitDegrees=45}
 {bone1Name=dyn_chain_B_5, bone2Name=dyn_chain_B_2, swing1LimitDegrees=45, swing2LimitDegrees=30, twistLimitDegrees=45}
 {bone1Name=dyn_chain_B_6, bone2Name=dyn_chain_B_5, swing1LimitDegrees=45, swing2LimitDegrees=30, twistLimitDegrees=45}
 {bone1Name=dyn_chain_C_1, bone2Name=head, swing1LimitDegrees=45, swing2LimitDegrees=45, twistLimitDegrees=5, twistMotion=Locked}
 {bone1Name=dyn_chain_C_2, bone2Name=dyn_chain_C_1, swing1LimitDegrees=45, swing2LimitDegrees=45, twistLimitDegrees=5, twistMotion=Locked}
 {bone1Name=dyn_chain_C_3, bone2Name=dyn_chain_C_2, swing1LimitDegrees=45, swing2LimitDegrees=45, twistLimitDegrees=5, twistMotion=Locked}
 {bone1Name=dyn_chain_D_1, bone2Name=head, swing1LimitDegrees=30, swing2LimitDegrees=10, twistLimitDegrees=5}
 {bone1Name=dyn_chain_D_2, bone2Name=dyn_chain_D_1, swing1LimitDegrees=30, swing2LimitDegrees=10, twistLimitDegrees=5}
 {bone1Name=dyn_chain_D_5, bone2Name=dyn_chain_D_2, swing1LimitDegrees=20, swing2LimitDegrees=10, twistLimitDegrees=5}
 {bone1Name=dyn_chain_D_6, bone2Name=dyn_chain_D_5, swing1LimitDegrees=20, swing2LimitDegrees=10, twistLimitDegrees=5}
 {bone1Name=dyn_chain_D_7, bone2Name=dyn_chain_D_6, swing1LimitDegrees=20, swing2LimitDegrees=10, twistLimitDegrees=5}
 {bone1Name=dyn_chain_F_2, bone2Name=dyn_chain_F_1, swing1LimitDegrees=10, swing2LimitDegrees=50, twistLimitDegrees=20}
 {bone1Name=dyn_chain_F_3, bone2Name=dyn_chain_F_2, swing1LimitDegrees=10, swing2LimitDegrees=50, twistLimitDegrees=20}
 {bone1Name=dyn_chain_F_4, bone2Name=dyn_chain_F_3, swing1LimitDegrees=10, swing2LimitDegrees=50, twistLimitDegrees=20}
 {bone1Name=dyn_chain_F_5, bone2Name=dyn_chain_F_3, swing1LimitDegrees=10, swing2LimitDegrees=50, twistLimitDegrees=20}
 {bone1Name=dyn_chain_G_1, bone2Name=dyn_chain_G_base, swing1LimitDegrees=10, swing2LimitDegrees=50, twistLimitDegrees=20}
 {bone1Name=dyn_chain_G_2, bone2Name=dyn_chain_G_1, swing1LimitDegrees=10, swing2LimitDegrees=50, twistLimitDegrees=20}
 {bone1Name=dyn_chain_G_3, bone2Name=dyn_chain_G_2, swing1LimitDegrees=10, swing2LimitDegrees=50, twistLimitDegrees=20}
 {bone1Name=dyn_chain_G_4, bone2Name=dyn_chain_G_3, swing1LimitDegrees=10, swing2LimitDegrees=50, twistLimitDegrees=20}
 {bone1Name=dyn_chain_G_5, bone2Name=dyn_chain_G_4, swing1LimitDegrees=10, swing2LimitDegrees=50, twistLimitDegrees=20}
 {bone1Name=dyn_chain_I_1, bone2Name=dyn_chain_G_base, swing1LimitDegrees=45, swing2LimitDegrees=45, twistLimitDegrees=45}
 {bone1Name=dyn_chain_I_2, bone2Name=dyn_chain_I_1, swing1LimitDegrees=45, swing2LimitDegrees=45, twistLimitDegrees=45}
 {bone1Name=dyn_chain_I_3, bone2Name=dyn_chain_I_2, swing1LimitDegrees=45, swing2LimitDegrees=45, twistLimitDegrees=45}
 {bone1Name=dyn_chain_I_4, bone2Name=dyn_chain_I_3, swing1LimitDegrees=45, swing2LimitDegrees=45, twistLimitDegrees=45}
 {bone1Name=dyn_chain_I_5, bone2Name=dyn_chain_I_4, swing1LimitDegrees=45, swing2LimitDegrees=45, twistLimitDegrees=45}
 {bone1Name=dyn_chain_I_6, bone2Name=dyn_chain_I_4, swing1LimitDegrees=45, swing2LimitDegrees=45, twistLimitDegrees=45}
 {bone1Name=dyn_chain_I_7, bone2Name=dyn_chain_I_3, swing1LimitDegrees=45, swing2LimitDegrees=45, twistLimitDegrees=45}
 {bone1Name=dyn_chain_J_1, bone2Name=head, swing1LimitDegrees=10, swing2LimitDegrees=20, twistLimitDegrees=5}
 {bone1Name=dyn_chain_J_2, bone2Name=dyn_chain_J_1, swing1LimitDegrees=10, swing2LimitDegrees=20, twistLimitDegrees=5}
 {bone1Name=dyn_chain_J_3, bone2Name=dyn_chain_J_2, swing1LimitDegrees=10, swing2LimitDegrees=20, twistLimitDegrees=5}
 {bone1Name=dyn_chain_J_4, bone2Name=dyn_chain_J_3, swing1LimitDegrees=10, swing2LimitDegrees=20, twistLimitDegrees=5}
 {bone1Name=dyn_chain_J_7, bone2Name=dyn_chain_J_2, swing1LimitDegrees=10, swing2LimitDegrees=20, twistLimitDegrees=5}
 {bone1Name=dyn_chain_J_8, bone2Name=dyn_chain_J_7, swing1LimitDegrees=10, swing2LimitDegrees=20, twistLimitDegrees=5}
 {bone1Name=dyn_chain_J_9, bone2Name=dyn_chain_J_8, swing1LimitDegrees=10, swing2LimitDegrees=20, twistLimitDegrees=5}
 {bone1Name=dyn_chain_J_10, bone2Name=dyn_chain_J_7, swing1LimitDegrees=10, swing2LimitDegrees=20, twistLimitDegrees=5}
 {bone1Name=dyn_chain_J_11, bone2Name=dyn_chain_J_10, swing1LimitDegrees=10, swing2LimitDegrees=20, twistLimitDegrees=5}
 {bone1Name=dyn_chain_J_12, bone2Name=dyn_chain_J_11, swing1LimitDegrees=10, swing2LimitDegrees=20, twistLimitDegrees=5}
 {bone1Name=dyn_chain_F_1, bone2Name=dyn_chain_G_1, swing1LimitDegrees=10, swing2LimitDegrees=50, twistLimitDegrees=20}
 {bone1Name=dyn_chain_G_base, bone2Name=head, swing1LimitDegrees=10, swing2LimitDegrees=50, twistLimitDegrees=20}
]
