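# Post-process-style secondary-motion AnimBP for the male SpringBreak face accessory mesh: beard, beads, strands, tufts and chains.
# It copies the parent mesh pose, then layers RigidBody simulation (standard and Frontend physics assets), per-bone spring AnimDynamics and bone constraints.
# The jawOpenLookingDown variable blends additive rotation into the upper beard bones.
# Weighted multi-target constraints partially bind chain bones to spine, neck, head and shoulder bones.
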AnimBlueprint /BRCosmetics/Characters/Player/Male/Medium/Bodies/M_MED_SpringBreak/Meshes/Parts/M_MED_SpringBreak_FaceAcc_AnimBP
target_skeleton=&/M_MED_SpringBreak_FaceAcc_Skeleton
graphs [
 AnimGraph {nodes [
   n1 = RigidBody!(
    ComponentPose=Local To Component(LocalPose=Copy Pose From Mesh() {bUseAttachedParent=true})
    Alpha
    OverrideWorldGravity:Vector="0,0,-980"
    ExternalForce
   ) -> (Pose) {
    bClampLinearTranslationLimitToRefPose=true
    bOverrideWorldGravity=true
    overridePhysicsAsset=&/M_MED_SpringBreak_FaceAcc_Physics
    simulationSpace=BaseBoneSpace
    alphaBoolBlend {blendOption=Linear}
    baseBoneRef {boneName=spine_04}
    componentAppliedLinearAccClamp=<5000 5000 5000>
    componentLinearAccScale=<0.3 0.3 0.08>
    componentLinearVelScale=<0.3 0.3 0.08>
    overrideWorldGravity {z=-980}
   }
   n2 = "Constraint - dyn_chain_A_1 (R)"(ComponentPose=@n1, Alpha=1, ConstraintWeights_0=1) {
    alphaBoolBlend {blendOption=Linear}
    boneToModify {boneName=dyn_chain_A_1}
    constraintSetup [{transformType=Rotation, targetBone {boneName=dyn_chain_B_1}}]
    constraintWeights [1]
   }
   n3 = "Constraint - dyn_chain_A_2 (R)"(ComponentPose=@n2, Alpha=1, ConstraintWeights_0=1) {
    alphaBoolBlend {blendOption=Linear}
    alphaScaleBiasClamp {scale=0.5}
    boneToModify {boneName=dyn_chain_A_2}
    constraintSetup [{transformType=Rotation, targetBone {boneName=dyn_chain_B_2}}]
    constraintWeights [1]
   }
   n4 = "Constraint - dyn_chain_C_4 (P)"(ComponentPose=@n3, Alpha=1, ConstraintWeights_0=1) {
    alphaBoolBlend {blendOption=Linear}
    boneToModify {boneName=dyn_chain_C_4}
    constraintSetup [{targetBone {boneName=dyn_chain_D_5}}]
    constraintWeights [1]
   }
   n5 = "Constraint - dyn_chain_H_4 (P)"(ComponentPose=@n4, Alpha=1, ConstraintWeights_0=1) {
    alphaBoolBlend {blendOption=Linear}
    boneToModify {boneName=dyn_chain_H_4}
    constraintSetup [{targetBone {boneName=dyn_chain_G_2}}]
    constraintWeights [1]
   }
   n6 = "Constraint - dyn_chain_F_7 (P)"(ComponentPose=@n5, Alpha=1, ConstraintWeights_0=1) {
    alphaBoolBlend {blendOption=Linear}
    boneToModify {boneName=dyn_chain_F_7}
    constraintSetup [{targetBone {boneName=dyn_chain_F_1}}]
    constraintWeights [1]
   }
   n7 = "Constraint - dyn_chain_F_8 (P)"(ComponentPose=@n6, Alpha=1, ConstraintWeights_0=1) {
    alphaBoolBlend {blendOption=Linear}
    boneToModify {boneName=dyn_chain_F_8}
    constraintSetup [{targetBone {boneName=dyn_chain_F_2}}]
    constraintWeights [1]
   }
   n8 = "Constraint - dyn_chain_F_9 (P)"(ComponentPose=@n7, Alpha=1, ConstraintWeights_0=1) {
    alphaBoolBlend {blendOption=Linear}
    boneToModify {boneName=dyn_chain_F_9}
    constraintSetup [{targetBone {boneName=dyn_chain_F_3}}]
    constraintWeights [1]
   }
   n9 = Anim Dynamics - Bone: dyn_L_beard_1!(ComponentPose=@n8, Alpha=1, GravityOverride:Vector="0,0,-980") -> (Pose) {
    angularDampingOverride=0.9999
    angularSpringConstant=2000
    bAngularSpring=true
    bOverrideAngularDamping=true
    bUseGravityOverride=true
    gravityScale=0.35
    simulationSpace=RootRelative
    alphaBoolBlend {blendOption=Linear}
    boundBone {boneName=dyn_L_beard_1}
    componentLinearAccScale=<1 1 1>
    componentLinearVelScale=<1 1 1>
    gravityOverride {z=-980}
    physicsBodyDefinitions [{
      boundBone {boneName=dyn_L_beard_1}
      boxExtents=<12 12 12>
      constraintSetup {angularLimitsMax=<10 20 50>, angularLimitsMin=<-10 -20 -50>, angularTarget=<1 0 0>}
      localJointOffset=<2 0 0>
     }]
   }
   n10 = Anim Dynamics - Bone: dyn_R_beard_1!(ComponentPose=@n9, Alpha=1, GravityOverride:Vector="0,0,-980") -> (Pose) {
    angularDampingOverride=0.9999
    angularSpringConstant=2000
    bAngularSpring=true
    bOverrideAngularDamping=true
    bUseGravityOverride=true
    gravityScale=0.35
    simulationSpace=RootRelative
    alphaBoolBlend {blendOption=Linear}
    boundBone {boneName=dyn_R_beard_1}
    componentLinearAccScale=<1 1 1>
    componentLinearVelScale=<1 1 1>
    gravityOverride {z=-980}
    physicsBodyDefinitions [{
      boundBone {boneName=dyn_R_beard_1}
      boxExtents=<14 14 14>
      constraintSetup {angularLimitsMax=<10 20 50>, angularLimitsMin=<-10 -20 -50>, angularTarget=<1 0 0>}
      localJointOffset=<2 0 0>
     }]
   }
   n11 = Anim Dynamics - Bone: dyn_L_beard_2!(ComponentPose=@n10, Alpha=1, GravityOverride:Vector="0,0,-980") -> (Pose) {
    angularDampingOverride=0.9999
    angularSpringConstant=2000
    bAngularSpring=true
    bOverrideAngularDamping=true
    bUseGravityOverride=true
    gravityScale=0.35
    simulationSpace=RootRelative
    alphaBoolBlend {blendOption=Linear}
    boundBone {boneName=dyn_L_beard_2}
    componentLinearAccScale=<1 1 1>
    componentLinearVelScale=<1 1 1>
    gravityOverride {z=-980}
    physicsBodyDefinitions [{
      boundBone {boneName=dyn_L_beard_2}
      boxExtents=<12 12 12>
      constraintSetup {angularLimitsMax=<10 20 50>, angularLimitsMin=<-10 -20 -50>, angularTarget=<1 0 0>}
      localJointOffset=<2 0 0>
     }]
   }
   n12 = Anim Dynamics - Bone: dyn_R_beard_2!(ComponentPose=@n11, Alpha=1, GravityOverride:Vector="0,0,-980") -> (Pose) {
    angularDampingOverride=0.9999
    angularSpringConstant=2000
    bAngularSpring=true
    bOverrideAngularDamping=true
    bUseGravityOverride=true
    gravityScale=0.35
    simulationSpace=RootRelative
    alphaBoolBlend {blendOption=Linear}
    boundBone {boneName=dyn_R_beard_2}
    componentLinearAccScale=<1 1 1>
    componentLinearVelScale=<1 1 1>
    gravityOverride {z=-980}
    physicsBodyDefinitions [{
      boundBone {boneName=dyn_R_beard_2}
      boxExtents=<14 14 14>
      constraintSetup {angularLimitsMax=<10 20 50>, angularLimitsMin=<-10 -20 -50>, angularTarget=<1 0 0>}
      localJointOffset=<2 0 0>
     }]
   }
   n13 = Get jawOpenLookingDown() -> (jawOpenLookingDown:double)
   n14 = "Transform (Modify) Bone - Bone: dyn_L_beard_1"!(ComponentPose=@n12, Alpha=@n13) -> (Pose) {
    rotationMode=BMM_Additive
    rotationSpace=BCS_BoneSpace
    alphaBoolBlend {blendOption=Linear}
    boneToModify {boneName=dyn_L_beard_1}
    rotation {yaw=35}
   }
   n15 = "Transform (Modify) Bone - Bone: dyn_R_beard_1"!(ComponentPose=@n14, Alpha=@n13) -> (Pose) {
    rotationMode=BMM_Additive
    rotationSpace=BCS_BoneSpace
    alphaBoolBlend {blendOption=Linear}
    boneToModify {boneName=dyn_R_beard_1}
    rotation {yaw=35}
   }
   n16 = Anim Dynamics - Bone: dyn_topBeads_1!(ComponentPose=@n15, Alpha=1, GravityOverride:Vector="0,0,-980") -> (Pose) {
    angularDampingOverride=0.9999
    angularSpringConstant=2000
    bAngularSpring=true
    bOverrideAngularDamping=true
    bUseGravityOverride=true
    gravityScale=0.35
    simulationSpace=RootRelative
    alphaBoolBlend {blendOption=Linear}
    boundBone {boneName=dyn_topBeads_1}
    componentLinearAccScale=<1 1 1>
    componentLinearVelScale=<1 1 1>
    gravityOverride {z=-980}
    physicsBodyDefinitions [{
      boundBone {boneName=dyn_topBeads_1}
      boxExtents=<12 12 12>
      constraintSetup {angularLimitsMax=<0 10 0>, angularLimitsMin=<0 -10 -15>, angularTarget=<1 0 0>}
      localJointOffset=<2 0 0>
     }]
   }
   n17 = Anim Dynamics - Bone: dyn_topBeads_2!(ComponentPose=@n16, Alpha=1, GravityOverride:Vector="0,0,-980") -> (Pose) {
    angularDampingOverride=0.9999
    angularSpringConstant=2000
    bAngularSpring=true
    bOverrideAngularDamping=true
    bUseGravityOverride=true
    gravityScale=0.35
    simulationSpace=RootRelative
    alphaBoolBlend {blendOption=Linear}
    boundBone {boneName=dyn_topBeads_2}
    componentLinearAccScale=<1 1 1>
    componentLinearVelScale=<1 1 1>
    gravityOverride {z=-980}
    physicsBodyDefinitions [{
      boundBone {boneName=dyn_topBeads_2}
      boxExtents=<12 12 12>
      constraintSetup {angularLimitsMax=<0 10 0>, angularLimitsMin=<0 -10 -15>, angularTarget=<1 0 0>}
      localJointOffset=<2 0 0>
     }]
   }
   n18 = Anim Dynamics - Bone: dyn_topBeads_3!(ComponentPose=@n17, Alpha=1, GravityOverride:Vector="0,0,-980") -> (Pose) {
    angularDampingOverride=0.9999
    angularSpringConstant=2000
    bAngularSpring=true
    bOverrideAngularDamping=true
    bUseGravityOverride=true
    gravityScale=0.35
    simulationSpace=RootRelative
    alphaBoolBlend {blendOption=Linear}
    boundBone {boneName=dyn_topBeads_3}
    componentLinearAccScale=<1 1 1>
    componentLinearVelScale=<1 1 1>
    gravityOverride {z=-980}
    physicsBodyDefinitions [{
      boundBone {boneName=dyn_topBeads_3}
      boxExtents=<12 12 12>
      constraintSetup {angularLimitsMax=<0 10 0>, angularLimitsMin=<0 -10 -15>, angularTarget=<1 0 0>}
      localJointOffset=<2 0 0>
     }]
   }
   n19 = Anim Dynamics - Bone: dyn_topBeads_4!(ComponentPose=@n18, Alpha=1, GravityOverride:Vector="0,0,-980") -> (Pose) {
    angularDampingOverride=0.9999
    angularSpringConstant=2000
    bAngularSpring=true
    bOverrideAngularDamping=true
    bUseGravityOverride=true
    gravityScale=0.35
    simulationSpace=RootRelative
    alphaBoolBlend {blendOption=Linear}
    boundBone {boneName=dyn_topBeads_4}
    componentLinearAccScale=<1 1 1>
    componentLinearVelScale=<1 1 1>
    gravityOverride {z=-980}
    physicsBodyDefinitions [{
      boundBone {boneName=dyn_topBeads_4}
      boxExtents=<12 12 12>
      constraintSetup {angularLimitsMax=<0 10 0>, angularLimitsMin=<0 -10 -15>, angularTarget=<1 0 0>}
      localJointOffset=<2 0 0>
     }]
   }
   n20 = Anim Dynamics - Bone: dyn_topBeads_5!(ComponentPose=@n19, Alpha=1, GravityOverride:Vector="0,0,-980") -> (Pose) {
    angularDampingOverride=0.9999
    angularSpringConstant=2000
    bAngularSpring=true
    bOverrideAngularDamping=true
    bUseGravityOverride=true
    gravityScale=0.35
    simulationSpace=RootRelative
    alphaBoolBlend {blendOption=Linear}
    boundBone {boneName=dyn_topBeads_5}
    componentLinearAccScale=<1 1 1>
    componentLinearVelScale=<1 1 1>
    gravityOverride {z=-980}
    physicsBodyDefinitions [{
      boundBone {boneName=dyn_topBeads_5}
      boxExtents=<12 12 12>
      constraintSetup {angularLimitsMax=<0 10 0>, angularLimitsMin=<0 -10 -15>, angularTarget=<1 0 0>}
      localJointOffset=<2 0 0>
     }]
   }
   n21 = Anim Dynamics - Bone: dyn_topBeads_6!(ComponentPose=@n20, Alpha=1, GravityOverride:Vector="0,0,-980") -> (Pose) {
    angularDampingOverride=0.9999
    angularSpringConstant=2000
    bAngularSpring=true
    bOverrideAngularDamping=true
    bUseGravityOverride=true
    gravityScale=0.35
    simulationSpace=RootRelative
    alphaBoolBlend {blendOption=Linear}
    boundBone {boneName=dyn_topBeads_6}
    componentLinearAccScale=<1 1 1>
    componentLinearVelScale=<1 1 1>
    gravityOverride {z=-980}
    physicsBodyDefinitions [{
      boundBone {boneName=dyn_topBeads_6}
      boxExtents=<12 12 12>
      constraintSetup {angularLimitsMax=<0 10 0>, angularLimitsMin=<0 -10 -15>, angularTarget=<1 0 0>}
      localJointOffset=<2 0 0>
     }]
   }
   n22 = Anim Dynamics - Bone: dyn_topBeads_7!(ComponentPose=@n21, Alpha=1, GravityOverride:Vector="0,0,-980") -> (Pose) {
    angularDampingOverride=0.9999
    angularSpringConstant=2000
    bAngularSpring=true
    bOverrideAngularDamping=true
    bUseGravityOverride=true
    gravityScale=0.35
    simulationSpace=RootRelative
    alphaBoolBlend {blendOption=Linear}
    boundBone {boneName=dyn_topBeads_7}
    componentLinearAccScale=<1 1 1>
    componentLinearVelScale=<1 1 1>
    gravityOverride {z=-980}
    physicsBodyDefinitions [{
      boundBone {boneName=dyn_topBeads_7}
      boxExtents=<12 12 12>
      constraintSetup {angularLimitsMax=<0 10 0>, angularLimitsMin=<0 -10 -15>, angularTarget=<1 0 0>}
      localJointOffset=<2 0 0>
     }]
   }
   n23 = Anim Dynamics - Bone: dyn_chain_topStrands_1!(ComponentPose=@n22, Alpha=1, GravityOverride:Vector="0,0,-980") -> (Pose) {
    angularDampingOverride=0.9999
    angularSpringConstant=2000
    bAngularSpring=true
    bOverrideAngularDamping=true
    bUseGravityOverride=true
    gravityScale=0.35
    simulationSpace=RootRelative
    alphaBoolBlend {blendOption=Linear}
    boundBone {boneName=dyn_chain_topStrands_1}
    componentLinearAccScale=<1 1 1>
    componentLinearVelScale=<1 1 1>
    gravityOverride {z=-980}
    physicsBodyDefinitions [{
      boundBone {boneName=dyn_chain_topStrands_1}
      boxExtents=<17 17 17>
      constraintSetup {angularLimitsMax=<60 45 0>, angularLimitsMin=<-60 -20 -45>, angularTarget=<1 0 0>}
      localJointOffset=<3 0 0>
     }]
   }
   n24 = Anim Dynamics - Bone: dyn_chain_topStrands_2!(ComponentPose=@n23, Alpha=1, GravityOverride:Vector="0,0,-980") -> (Pose) {
    angularDampingOverride=0.9999
    angularSpringConstant=2000
    bAngularSpring=true
    bOverrideAngularDamping=true
    bUseGravityOverride=true
    gravityScale=0.35
    simulationSpace=RootRelative
    alphaBoolBlend {blendOption=Linear}
    boundBone {boneName=dyn_chain_topStrands_2}
    componentLinearAccScale=<1 1 1>
    componentLinearVelScale=<1 1 1>
    gravityOverride {z=-980}
    physicsBodyDefinitions [{
      boundBone {boneName=dyn_chain_topStrands_2}
      boxExtents=<17 17 17>
      constraintSetup {angularLimitsMax=<60 45 0>, angularLimitsMin=<-60 -20 -45>, angularTarget=<1 0 0>}
      localJointOffset=<3 0 0>
     }]
   }
   n25 = Anim Dynamics - Bone: dyn_chain_topStrands_3!(ComponentPose=@n24, Alpha=1, GravityOverride:Vector="0,0,-980") -> (Pose) {
    angularDampingOverride=0.9999
    angularSpringConstant=2000
    bAngularSpring=true
    bOverrideAngularDamping=true
    bUseGravityOverride=true
    gravityScale=0.35
    simulationSpace=RootRelative
    alphaBoolBlend {blendOption=Linear}
    boundBone {boneName=dyn_chain_topStrands_3}
    componentLinearAccScale=<1 1 1>
    componentLinearVelScale=<1 1 1>
    gravityOverride {z=-980}
    physicsBodyDefinitions [{
      boundBone {boneName=dyn_chain_topStrands_3}
      boxExtents=<17 17 17>
      constraintSetup {angularLimitsMax=<60 45 0>, angularLimitsMin=<-60 -20 -45>, angularTarget=<1 0 0>}
      localJointOffset=<3 0 0>
     }]
   }
   n26 = Anim Dynamics - Bone: dyn_chain_topStrands_4!(ComponentPose=@n25, Alpha=1, GravityOverride:Vector="0,0,-980") -> (Pose) {
    angularDampingOverride=0.9999
    angularSpringConstant=2000
    bAngularSpring=true
    bOverrideAngularDamping=true
    bUseGravityOverride=true
    gravityScale=0.35
    simulationSpace=RootRelative
    alphaBoolBlend {blendOption=Linear}
    boundBone {boneName=dyn_chain_topStrands_4}
    componentLinearAccScale=<1 1 1>
    componentLinearVelScale=<1 1 1>
    gravityOverride {z=-980}
    physicsBodyDefinitions [{
      boundBone {boneName=dyn_chain_topStrands_4}
      boxExtents=<17 17 17>
      constraintSetup {angularLimitsMax=<60 45 0>, angularLimitsMin=<-60 -20 -45>, angularTarget=<1 0 0>}
      localJointOffset=<3 0 0>
     }]
   }
   n27 = "Constraint - dyn_chain_topStrands_1 (P)"(ComponentPose=@n26, Alpha=1, ConstraintWeights_0=1) {
    alphaBoolBlend {blendOption=Linear}
    alphaScaleBiasClamp {scale=0.5}
    boneToModify {boneName=dyn_chain_topStrands_1}
    constraintSetup [{targetBone {boneName=dyn_chain_B_1}}]
    constraintWeights [1]
   }
   n28 = "Constraint - dyn_chain_topStrands_4 (R)"(ComponentPose=@n27, Alpha=1, ConstraintWeights_0=1) {
    alphaBoolBlend {blendOption=Linear}
    alphaScaleBiasClamp {scale=0.75}
    boneToModify {boneName=dyn_chain_topStrands_4}
    constraintSetup [{transformType=Rotation, targetBone {boneName=dyn_chain_B_2}}]
    constraintWeights [1]
   }
   n29 = Anim Dynamics - Bone: dyn_R_front_tuft!(ComponentPose=@n28, Alpha=1, GravityOverride:Vector="0,0,-980") -> (Pose) {
    angularDampingOverride=1
    angularSpringConstant=200
    bAngularSpring=true
    bOverrideAngularDamping=true
    bUseGravityOverride=true
    gravityScale=0.35
    simulationSpace=RootRelative
    alphaBoolBlend {blendOption=Linear}
    boundBone {boneName=dyn_R_front_tuft}
    componentLinearAccScale=<1 1 1>
    componentLinearVelScale=<1 1 1>
    gravityOverride {z=-980}
    physicsBodyDefinitions [{
      boundBone {boneName=dyn_R_front_tuft}
      boxExtents=<14 14 14>
      constraintSetup {angularLimitsMax=<0 0 0>, angularLimitsMin=<0 0 -30>, angularTarget=<1 0 0>}
      localJointOffset=<2 0 0>
     }]
   }
   n30 = Anim Dynamics - Bone: dyn_L_front_tuft!(ComponentPose=@n29, Alpha=1, GravityOverride:Vector="0,0,-980") -> (Pose) {
    angularDampingOverride=1
    angularSpringConstant=200
    bAngularSpring=true
    bOverrideAngularDamping=true
    bUseGravityOverride=true
    gravityScale=0.35
    simulationSpace=RootRelative
    alphaBoolBlend {blendOption=Linear}
    boundBone {boneName=dyn_L_front_tuft}
    componentLinearAccScale=<1 1 1>
    componentLinearVelScale=<1 1 1>
    gravityOverride {z=-980}
    physicsBodyDefinitions [{
      boundBone {boneName=dyn_L_front_tuft}
      boxExtents=<14 14 14>
      constraintSetup {angularLimitsMax=<0 0 30>, angularLimitsMin=<0 0 0>, angularTarget=<1 0 0>}
      localJointOffset=<2 0 0>
     }]
   }
   n31 = Anim Dynamics - Bone: dyn_chain_J_dangly_1!(ComponentPose=@n30, Alpha=1, GravityOverride:Vector="0,0,-980") -> (Pose) {
    angularDampingOverride=0.9999
    angularSpringConstant=2000
    bAngularSpring=true
    bOverrideAngularDamping=true
    bUseGravityOverride=true
    gravityScale=0.35
    simulationSpace=RootRelative
    alphaBoolBlend {blendOption=Linear}
    boundBone {boneName=dyn_chain_J_dangly_1}
    componentLinearAccScale=<1 1 1>
    componentLinearVelScale=<1 1 1>
    gravityOverride {z=-980}
    physicsBodyDefinitions [{
      boundBone {boneName=dyn_chain_J_dangly_1}
      boxExtents=<12 12 12>
      constraintSetup {angularLimitsMax=<0 30 5>, angularLimitsMin=<0 0 -5>, angularTarget=<1 0 0>}
      localJointOffset=<2 0 0>
     }]
   }
   n32 = Anim Dynamics - Bone: dyn_chain_J_dangly_2!(ComponentPose=@n31, Alpha=1, GravityOverride:Vector="0,0,-980") -> (Pose) {
    angularDampingOverride=0.9999
    angularSpringConstant=2000
    bAngularSpring=true
    bOverrideAngularDamping=true
    bUseGravityOverride=true
    gravityScale=0.35
    simulationSpace=RootRelative
    alphaBoolBlend {blendOption=Linear}
    boundBone {boneName=dyn_chain_J_dangly_2}
    componentLinearAccScale=<1 1 1>
    componentLinearVelScale=<1 1 1>
    gravityOverride {z=-980}
    physicsBodyDefinitions [{
      boundBone {boneName=dyn_chain_J_dangly_2}
      boxExtents=<12 12 12>
      constraintSetup {angularLimitsMax=<0 30 5>, angularLimitsMin=<0 0 -5>, angularTarget=<1 0 0>}
      localJointOffset=<2 0 0>
     }]
   }
   n33 = Anim Dynamics - Bone: dyn_chain_J_dangly_3!(ComponentPose=@n32, Alpha=1, GravityOverride:Vector="0,0,-980") -> (Pose) {
    angularDampingOverride=0.9999
    angularSpringConstant=2000
    bAngularSpring=true
    bOverrideAngularDamping=true
    bUseGravityOverride=true
    gravityScale=0.35
    simulationSpace=RootRelative
    alphaBoolBlend {blendOption=Linear}
    boundBone {boneName=dyn_chain_J_dangly_3}
    componentLinearAccScale=<1 1 1>
    componentLinearVelScale=<1 1 1>
    gravityOverride {z=-980}
    physicsBodyDefinitions [{
      boundBone {boneName=dyn_chain_J_dangly_3}
      boxExtents=<12 12 12>
      constraintSetup {angularLimitsMax=<0 30 5>, angularLimitsMin=<0 0 -5>, angularTarget=<1 0 0>}
      localJointOffset=<2 0 0>
     }]
   }
   n34 = "Constraint - dyn_chain_J_dangly_2 (T)"(ComponentPose=@n33, Alpha=1, ConstraintWeights_0=1) {
    alphaBoolBlend {blendOption=Linear}
    alphaScaleBiasClamp {scale=0.3}
    boneToModify {boneName=dyn_chain_J_dangly_2}
    constraintSetup [{transformType=Translation, targetBone {boneName=dyn_chain_J_10}}]
    constraintWeights [1]
   }
   n35 = "Constraint - dyn_chain_J_dangly_2 (R)"(ComponentPose=@n34, Alpha=1, ConstraintWeights_0=1) {
    alphaBoolBlend {blendOption=Linear}
    boneToModify {boneName=dyn_chain_J_dangly_2}
    constraintSetup [{transformType=Rotation, targetBone {boneName=dyn_chain_J_10}}]
    constraintWeights [1]
   }
   n36 = "Constraint - dyn_chain_J_dangly_3 (TT)"!(ComponentPose=@n35, Alpha=1, ConstraintWeights_0=0.6, ConstraintWeights_1:float=0.4) -> (Pose) {
    alphaBoolBlend {blendOption=Linear}
    alphaScaleBiasClamp {scale=0.3}
    boneToModify {boneName=dyn_chain_J_dangly_3}
    constraintSetup [{transformType=Translation, targetBone {boneName=dyn_chain_J_10}}, {transformType=Translation, targetBone {boneName=dyn_chain_J_7}}]
    constraintWeights [1,1]
   }
   n37 = "Constraint - dyn_chain_J_dangly_3 (RR)"!(ComponentPose=@n36, Alpha=1, ConstraintWeights_0=0.6, ConstraintWeights_1:float=0.4) -> (Pose) {
    alphaBoolBlend {blendOption=Linear}
    boneToModify {boneName=dyn_chain_J_dangly_3}
    constraintSetup [{transformType=Rotation, targetBone {boneName=dyn_chain_J_10}}, {transformType=Rotation, targetBone {boneName=dyn_chain_J_7}}]
    constraintWeights [1,1]
   }
   n38 = Anim Dynamics - Bone: dyn_chain_A_1!(ComponentPose=@n37, Alpha=1, GravityOverride:Vector="0,0,-980") -> (Pose) {
    angularDampingOverride=0.9999
    angularSpringConstant=500
    bAngularSpring=true
    bOverrideAngularDamping=true
    bUseGravityOverride=true
    gravityScale=0.35
    simulationSpace=RootRelative
    alphaBoolBlend {blendOption=Linear}
    boundBone {boneName=dyn_chain_A_1}
    componentLinearAccScale=<0.7 0.7 0.7>
    componentLinearVelScale=<0.7 0.7 0.7>
    gravityOverride {z=-980}
    physicsBodyDefinitions [{
      boundBone {boneName=dyn_chain_A_1}
      boxExtents=<17 17 17>
      constraintSetup {angularLimitsMax=<0 30 0>, angularLimitsMin=<0 -5 -60>, angularTarget=<1 0 0>}
      localJointOffset=<2 0 0>
     }]
   }
   n39 = Anim Dynamics - Bone: dyn_chain_A_2!(ComponentPose=@n38, Alpha=1, GravityOverride:Vector="0,0,-980") -> (Pose) {
    angularDampingOverride=0.9999
    angularSpringConstant=500
    bAngularSpring=true
    bOverrideAngularDamping=true
    bUseGravityOverride=true
    gravityScale=0.35
    simulationSpace=RootRelative
    alphaBoolBlend {blendOption=Linear}
    boundBone {boneName=dyn_chain_A_2}
    componentLinearAccScale=<0.7 0.7 0.7>
    componentLinearVelScale=<0.7 0.7 0.7>
    gravityOverride {z=-980}
    physicsBodyDefinitions [{
      boundBone {boneName=dyn_chain_A_2}
      boxExtents=<17 17 17>
      constraintSetup {angularLimitsMax=<0 30 0>, angularLimitsMin=<0 -5 -60>, angularTarget=<1 0 0>}
      localJointOffset=<2 0 0>
     }]
   }
   n40 = Anim Dynamics - Bone: dyn_chain_A_3!(ComponentPose=@n39, Alpha=1, GravityOverride:Vector="0,0,-980") -> (Pose) {
    angularDampingOverride=0.9999
    angularSpringConstant=500
    bAngularSpring=true
    bOverrideAngularDamping=true
    bUseGravityOverride=true
    gravityScale=0.35
    simulationSpace=RootRelative
    alphaBoolBlend {blendOption=Linear}
    boundBone {boneName=dyn_chain_A_3}
    componentLinearAccScale=<0.7 0.7 0.7>
    componentLinearVelScale=<0.7 0.7 0.7>
    gravityOverride {z=-980}
    physicsBodyDefinitions [{
      boundBone {boneName=dyn_chain_A_3}
      boxExtents=<17 17 17>
      constraintSetup {angularLimitsMax=<0 30 0>, angularLimitsMin=<0 -5 -60>, angularTarget=<1 0 0>}
      localJointOffset=<2 0 0>
     }]
   }
   n41 = Anim Dynamics - Bone: dyn_chain_C_4!(ComponentPose=@n40, Alpha=1, GravityOverride:Vector="0,0,-980") -> (Pose) {
    angularDampingOverride=0.9999
    angularSpringConstant=2000
    bAngularSpring=true
    bOverrideAngularDamping=true
    bUseGravityOverride=true
    gravityScale=0.35
    simulationSpace=RootRelative
    alphaBoolBlend {blendOption=Linear}
    boundBone {boneName=dyn_chain_C_4}
    componentLinearAccScale=<1 1 1>
    componentLinearVelScale=<1 1 1>
    gravityOverride {z=-980}
    physicsBodyDefinitions [{
      boundBone {boneName=dyn_chain_C_4}
      boxExtents=<12 12 12>
      constraintSetup {angularLimitsMax=<0 30 10>, angularLimitsMin=<0 0 -10>, angularTarget=<1 0 0>}
      localJointOffset=<2 0 0>
     }]
   }
   n42 = Anim Dynamics - Bone: dyn_chain_E_2!(ComponentPose=@n41, Alpha=1, GravityOverride:Vector="0,0,-980") -> (Pose) {
    angularDampingOverride=0.9999
    angularSpringConstant=500
    bAngularSpring=true
    bOverrideAngularDamping=true
    bUseGravityOverride=true
    gravityScale=0.35
    simulationSpace=RootRelative
    alphaBoolBlend {blendOption=Linear}
    boundBone {boneName=dyn_chain_E_2}
    componentLinearAccScale=<0.7 0.7 0.7>
    componentLinearVelScale=<0.7 0.7 0.7>
    gravityOverride {z=-980}
    physicsBodyDefinitions [{
      boundBone {boneName=dyn_chain_E_2}
      boxExtents=<17 17 17>
      constraintSetup {angularLimitsMax=<0 0 0>, angularLimitsMin=<0 -30 -60>, angularTarget=<1 0 0>}
      localJointOffset=<2 0 0>
     }]
   }
   n43 = Anim Dynamics - Bone: dyn_chain_E_3!(ComponentPose=@n42, Alpha=1, GravityOverride:Vector="0,0,-980") -> (Pose) {
    angularDampingOverride=0.9999
    angularSpringConstant=500
    bAngularSpring=true
    bOverrideAngularDamping=true
    bUseGravityOverride=true
    gravityScale=0.35
    simulationSpace=RootRelative
    alphaBoolBlend {blendOption=Linear}
    boundBone {boneName=dyn_chain_E_3}
    componentLinearAccScale=<0.7 0.7 0.7>
    componentLinearVelScale=<0.7 0.7 0.7>
    gravityOverride {z=-980}
    physicsBodyDefinitions [{
      boundBone {boneName=dyn_chain_E_3}
      boxExtents=<17 17 17>
      constraintSetup {angularLimitsMax=<0 0 0>, angularLimitsMin=<0 -30 -60>, angularTarget=<1 0 0>}
      localJointOffset=<2 0 0>
     }]
   }
   n44 = Anim Dynamics - Bone: dyn_chain_F_7!(ComponentPose=@n43, Alpha=1, GravityOverride:Vector="0,0,-980") -> (Pose) {
    angularDampingOverride=0.9999
    angularSpringConstant=500
    bAngularSpring=true
    bOverrideAngularDamping=true
    bUseGravityOverride=true
    gravityScale=0.35
    simulationSpace=RootRelative
    alphaBoolBlend {blendOption=Linear}
    boundBone {boneName=dyn_chain_F_7}
    componentLinearAccScale=<0.5 0.5 0.5>
    componentLinearVelScale=<0.5 0.5 0.5>
    gravityOverride {z=-980}
    physicsBodyDefinitions [{
      boundBone {boneName=dyn_chain_F_7}
      boxExtents=<15 15 15>
      constraintSetup {angularLimitsMax=<0 -1 5>, angularLimitsMin=<0 -30 -10>, angularTarget=<1 0 0>}
      localJointOffset=<2 0 0>
     }]
   }
   n45 = Anim Dynamics - Bone: dyn_chain_F_8!(ComponentPose=@n44, Alpha=1, GravityOverride:Vector="0,0,-980") -> (Pose) {
    angularDampingOverride=0.9999
    angularSpringConstant=500
    bAngularSpring=true
    bOverrideAngularDamping=true
    bUseGravityOverride=true
    gravityScale=0.35
    simulationSpace=RootRelative
    alphaBoolBlend {blendOption=Linear}
    boundBone {boneName=dyn_chain_F_8}
    componentLinearAccScale=<0.5 0.5 0.5>
    componentLinearVelScale=<0.5 0.5 0.5>
    gravityOverride {z=-980}
    physicsBodyDefinitions [{
      boundBone {boneName=dyn_chain_F_8}
      boxExtents=<15 15 15>
      constraintSetup {angularLimitsMax=<0 -1 5>, angularLimitsMin=<0 -30 -10>, angularTarget=<1 0 0>}
      localJointOffset=<2 0 0>
     }]
   }
   n46 = Anim Dynamics - Bone: dyn_chain_F_9!(ComponentPose=@n45, Alpha=1, GravityOverride:Vector="0,0,-980") -> (Pose) {
    angularDampingOverride=0.9999
    angularSpringConstant=500
    bAngularSpring=true
    bOverrideAngularDamping=true
    bUseGravityOverride=true
    gravityScale=0.35
    simulationSpace=RootRelative
    alphaBoolBlend {blendOption=Linear}
    boundBone {boneName=dyn_chain_F_9}
    componentLinearAccScale=<0.5 0.5 0.5>
    componentLinearVelScale=<0.5 0.5 0.5>
    gravityOverride {z=-980}
    physicsBodyDefinitions [{
      boundBone {boneName=dyn_chain_F_9}
      boxExtents=<15 15 15>
      constraintSetup {angularLimitsMax=<0 -1 5>, angularLimitsMin=<0 -30 -10>, angularTarget=<1 0 0>}
      localJointOffset=<2 0 0>
     }]
   }
   n47 = Anim Dynamics - Bone: dyn_chain_H_2!(ComponentPose=@n46, Alpha=1, GravityOverride:Vector="0,0,-980") -> (Pose) {
    angularDampingOverride=0.999
    angularSpringConstant=7500
    bAngularSpring=true
    bOverrideAngularDamping=true
    bUseGravityOverride=true
    gravityScale=0.35
    simulationSpace=RootRelative
    alphaBoolBlend {blendOption=Linear}
    boundBone {boneName=dyn_chain_H_2}
    componentLinearAccScale=<0.5 0.5 0.5>
    componentLinearVelScale=<0.5 0.5 0.5>
    gravityOverride {z=-980}
    physicsBodyDefinitions [{
      boundBone {boneName=dyn_chain_H_2}
      boxExtents=<29 29 29>
      constraintSetup {angularLimitsMax=<0 0 0>, angularLimitsMin=<0 -30 0>, angularTarget=<1 0 0>}
      localJointOffset=<3 0 0>
     }]
   }
   n48 = Anim Dynamics - Bone: dyn_chain_H_3!(ComponentPose=@n47, Alpha=1, GravityOverride:Vector="0,0,-980") -> (Pose) {
    angularDampingOverride=0.999
    angularSpringConstant=7500
    bAngularSpring=true
    bOverrideAngularDamping=true
    bUseGravityOverride=true
    gravityScale=0.35
    simulationSpace=RootRelative
    alphaBoolBlend {blendOption=Linear}
    boundBone {boneName=dyn_chain_H_3}
    componentLinearAccScale=<0.5 0.5 0.5>
    componentLinearVelScale=<0.5 0.5 0.5>
    gravityOverride {z=-980}
    physicsBodyDefinitions [{
      boundBone {boneName=dyn_chain_H_3}
      boxExtents=<29 29 29>
      constraintSetup {angularLimitsMax=<0 0 0>, angularLimitsMin=<0 -30 0>, angularTarget=<1 0 0>}
      localJointOffset=<3 0 0>
     }]
   }
   n49 = Anim Dynamics - Bone: dyn_chain_H_4!(ComponentPose=@n48, Alpha=1, GravityOverride:Vector="0,0,-980") -> (Pose) {
    angularDampingOverride=0.999
    angularSpringConstant=7500
    bAngularSpring=true
    bOverrideAngularDamping=true
    bUseGravityOverride=true
    gravityScale=0.35
    simulationSpace=RootRelative
    alphaBoolBlend {blendOption=Linear}
    boundBone {boneName=dyn_chain_H_4}
    componentLinearAccScale=<0.5 0.5 0.5>
    componentLinearVelScale=<0.5 0.5 0.5>
    gravityOverride {z=-980}
    physicsBodyDefinitions [{
      boundBone {boneName=dyn_chain_H_4}
      boxExtents=<29 29 29>
      constraintSetup {angularLimitsMax=<0 0 0>, angularLimitsMin=<0 -30 0>, angularTarget=<1 0 0>}
      localJointOffset=<3 0 0>
     }]
   }
   n50 = Anim Dynamics - Bone: dyn_chain_H_5!(ComponentPose=@n49, Alpha=1, GravityOverride:Vector="0,0,-980") -> (Pose) {
    angularDampingOverride=0.999
    angularSpringConstant=7500
    bAngularSpring=true
    bOverrideAngularDamping=true
    bUseGravityOverride=true
    gravityScale=0.35
    simulationSpace=RootRelative
    alphaBoolBlend {blendOption=Linear}
    boundBone {boneName=dyn_chain_H_5}
    componentLinearAccScale=<0.5 0.5 0.5>
    componentLinearVelScale=<0.5 0.5 0.5>
    gravityOverride {z=-980}
    physicsBodyDefinitions [{
      boundBone {boneName=dyn_chain_H_5}
      boxExtents=<29 29 29>
      constraintSetup {angularLimitsMax=<0 0 0>, angularLimitsMin=<0 -30 0>, angularTarget=<1 0 0>}
      localJointOffset=<3 0 0>
     }]
   }
   n51 = "Transform (Modify) Bone - Bone: dyn_chain_G_base"!(ComponentPose=@n50, Alpha=1, Scale="1,1,1") -> (Pose) {
    rotationMode=BMM_Additive
    rotationSpace=BCS_BoneSpace
    translationMode=BMM_Additive
    translationSpace=BCS_BoneSpace
    alphaBoolBlend {blendOption=Linear}
    boneToModify {boneName=dyn_chain_G_base}
    translation {x=-4, z=2}
   }
   n52 = RigidBody(ComponentPose=@n51, OverrideWorldGravity="0,0,-980") {
    bClampLinearTranslationLimitToRefPose=true
    bOverrideWorldGravity=true
    overridePhysicsAsset=&/M_MED_SpringBreak_FaceAcc_Physics_Frontend
    simulationSpace=BaseBoneSpace
    alphaBoolBlend {blendOption=Linear}
    baseBoneRef {boneName=spine_04}
    componentAppliedLinearAccClamp=<5000 5000 5000>
    componentLinearAccScale=<0.3 0.3 0.08>
    componentLinearVelScale=<0.3 0.3 0.08>
    overrideWorldGravity {z=-980}
   }
   n53 = "Constraint - dyn_chain_G_base (PPPPP)"!(
    ComponentPose=@n52
    Alpha=1
    ConstraintWeights_0=0.2733
    ConstraintWeights_1:float=0.3463
    ConstraintWeights_2:float=0.3404
    ConstraintWeights_3:float=0.0094
    ConstraintWeights_4:float=0.0306
   ) -> (Pose) {
    alphaBoolBlend {blendOption=Linear}
    alphaScaleBiasClamp {scale=0.2}
    boneToModify {boneName=dyn_chain_G_base}
    constraintSetup [
     {targetBone {boneName=neck_02}}
     {targetBone {boneName=head}}
     {targetBone {boneName=neck_01}}
     {targetBone {boneName=spine_04}}
     {targetBone {boneName=spine_05}}
    ]
    constraintWeights [1,1,1,1,1]
   }
   n54 = "Constraint - dyn_chain_G_1 (PPPPP)"!(
    ComponentPose=@n53
    Alpha=1
    ConstraintWeights_0=0.2733
    ConstraintWeights_1:float=0.3463
    ConstraintWeights_2:float=0.3404
    ConstraintWeights_3:float=0.0094
    ConstraintWeights_4:float=0.0306
   ) -> (Pose) {
    alphaBoolBlend {blendOption=Linear}
    alphaScaleBiasClamp {scale=0.2}
    boneToModify {boneName=dyn_chain_G_1}
    constraintSetup [
     {targetBone {boneName=neck_02}}
     {targetBone {boneName=head}}
     {targetBone {boneName=neck_01}}
     {targetBone {boneName=spine_04}}
     {targetBone {boneName=spine_05}}
    ]
    constraintWeights [1,1,1,1,1]
   }
   n55 = "Constraint - dyn_chain_G_2 (PPPPPPP)"!(
    ComponentPose=@n54
    Alpha=1
    ConstraintWeights_0=0.0842
    ConstraintWeights_1:float=0.0895
    ConstraintWeights_2:float=0.015
    ConstraintWeights_3:float=0.0148
    ConstraintWeights_4:float=0.5599
    ConstraintWeights_5:float=0.1712
    ConstraintWeights_6:float=0.0655
   ) -> (Pose) {
    alphaBoolBlend {blendOption=Linear}
    alphaScaleBiasClamp {scale=0.2}
    boneToModify {boneName=dyn_chain_G_2}
    constraintSetup [
     {targetBone {boneName=spine_04}}
     {targetBone {boneName=spine_05}}
     {targetBone {boneName=clavicle_l}}
     {targetBone {boneName=clavicle_r}}
     {targetBone {boneName=neck_01}}
     {targetBone {boneName=neck_02}}
     {targetBone {boneName=head}}
    ]
    constraintWeights [1,1,1,1,1,1,1]
   }
   n56 = "Constraint - dyn_chain_G_3 (PPPPPPPP)"!(
    ComponentPose=@n55
    Alpha=1
    ConstraintWeights_0=0.0134
    ConstraintWeights_1:float=0.091
    ConstraintWeights_2:float=0.0603
    ConstraintWeights_3:float=0.475
    ConstraintWeights_4:float=0.1522
    ConstraintWeights_5:float=0.025
    ConstraintWeights_6:float=0.0933
    ConstraintWeights_7:float=0.0898
   ) -> (Pose) {
    alphaBoolBlend {blendOption=Linear}
    alphaScaleBiasClamp {scale=0.2}
    boneToModify {boneName=dyn_chain_G_3}
    constraintSetup [
     {targetBone {boneName=upperarm_l}}
     {targetBone {boneName=neck_01}}
     {targetBone {boneName=spine_03}}
     {targetBone {boneName=spine_05}}
     {targetBone {boneName=spine_04}}
     {targetBone {boneName=upperarm_r}}
     {targetBone {boneName=clavicle_r}}
     {targetBone {boneName=clavicle_l}}
    ]
    constraintWeights [1,1,1,1,1,1,1,1]
   }
   n57 = "Constraint - dyn_chain_F_2 (PPPPPP)"!(
    ComponentPose=@n56
    Alpha=1
    ConstraintWeights_0=0.1298
    ConstraintWeights_1:float=0.5077
    ConstraintWeights_2:float=0.2273
    ConstraintWeights_3:float=0.0004
    ConstraintWeights_4:float=0.1345
    ConstraintWeights_5:float=0.0003
   ) -> (Pose) {
    alphaBoolBlend {blendOption=Linear}
    alphaScaleBiasClamp {scale=0.6}
    boneToModify {boneName=dyn_chain_F_2}
    constraintSetup [
     {targetBone {boneName=neck_01}}
     {targetBone {boneName=spine_05}}
     {targetBone {boneName=spine_04}}
     {targetBone {boneName=upperarm_r}}
     {targetBone {boneName=clavicle_r}}
     {targetBone {boneName=clavicle_l}}
    ]
    constraintWeights [1,1,1,1,1,1]
   }
   n58 = "Constraint - dyn_chain_F_3 (PPPPPPP)"!(
    ComponentPose=@n57
    Alpha=1
    ConstraintWeights_0=0.0124
    ConstraintWeights_1:float=0.1276
    ConstraintWeights_2:float=0.0708
    ConstraintWeights_3:float=0.3812
    ConstraintWeights_4:float=0.0106
    ConstraintWeights_5:float=0.3038
    ConstraintWeights_6:float=0.0936
   ) -> (Pose) {
    alphaBoolBlend {blendOption=Linear}
    alphaScaleBiasClamp {scale=0.6}
    boneToModify {boneName=dyn_chain_F_3}
    constraintSetup [
     {targetBone {boneName=neck_01}}
     {targetBone {boneName=spine_04}}
     {targetBone {boneName=spine_03}}
     {targetBone {boneName=clavicle_r}}
     {targetBone {boneName=spine_02}}
     {targetBone {boneName=spine_05}}
     {targetBone {boneName=upperarm_r}}
    ]
    constraintWeights [1,1,1,1,1,1,1]
   }
   n59 = "Constraint - dyn_chain_F_5 (PPPPPPPP)"!(
    ComponentPose=@n58
    Alpha=1
    ConstraintWeights_0=0.308
    ConstraintWeights_1:float=0.2841
    ConstraintWeights_2:float=0.2164
    ConstraintWeights_3:float=0.0147
    ConstraintWeights_4:float=0.004
    ConstraintWeights_5:float=0.0082
    ConstraintWeights_6:float=0.0456
    ConstraintWeights_7:float=0.119
   ) -> (Pose) {
    alphaBoolBlend {blendOption=Linear}
    alphaScaleBiasClamp {scale=0.6}
    boneToModify {boneName=dyn_chain_F_5}
    constraintSetup [
     {targetBone {boneName=spine_03}}
     {targetBone {boneName=spine_05}}
     {targetBone {boneName=spine_04}}
     {targetBone {boneName=spine_01}}
     {targetBone {boneName=pelvis}}
     {targetBone {boneName=upperarm_r}}
     {targetBone {boneName=clavicle_r}}
     {targetBone {boneName=spine_02}}
    ]
    constraintWeights [1,1,1,1,1,1,1,1]
   }
   n60 = "Constraint - dyn_chain_F_4 (PPPPPPPP)"!(
    ComponentPose=@n59
    Alpha=1
    ConstraintWeights_0=0.2837
    ConstraintWeights_1:float=0.2934
    ConstraintWeights_2:float=0.2229
    ConstraintWeights_3:float=0.0101
    ConstraintWeights_4:float=0.0027
    ConstraintWeights_5:float=0.0157
    ConstraintWeights_6:float=0.0865
    ConstraintWeights_7:float=0.0851
   ) -> (Pose) {
    alphaBoolBlend {blendOption=Linear}
    alphaScaleBiasClamp {scale=0.6}
    boneToModify {boneName=dyn_chain_F_4}
    constraintSetup [
     {targetBone {boneName=spine_03}}
     {targetBone {boneName=spine_05}}
     {targetBone {boneName=spine_04}}
     {targetBone {boneName=spine_01}}
     {targetBone {boneName=pelvis}}
     {targetBone {boneName=upperarm_r}}
     {targetBone {boneName=clavicle_r}}
     {targetBone {boneName=spine_02}}
    ]
    constraintWeights [1,1,1,1,1,1,1,1]
   }
   n61 = "Constraint - dyn_chain_I_2 (PPPPPP)"!(
    ComponentPose=@n60
    Alpha=1
    ConstraintWeights_0=0.0121
    ConstraintWeights_1:float=0.0423
    ConstraintWeights_2:float=0.0044
    ConstraintWeights_3:float=0.3612
    ConstraintWeights_4:float=0.321
    ConstraintWeights_5:float=0.259
   ) -> (Pose) {
    alphaBoolBlend {blendOption=Linear}
    alphaScaleBiasClamp {scale=0.6}
    boneToModify {boneName=dyn_chain_I_2}
    constraintSetup [
     {targetBone {boneName=spine_04}}
     {targetBone {boneName=spine_05}}
     {targetBone {boneName=clavicle_l}}
     {targetBone {boneName=neck_01}}
     {targetBone {boneName=neck_02}}
     {targetBone {boneName=head}}
    ]
    constraintWeights [1,1,1,1,1,1]
   }
   n62 = "Constraint - dyn_chain_I_3 (PPPPPPP)"!(
    ComponentPose=@n61
    Alpha=1
    ConstraintWeights_0=0.001
    ConstraintWeights_1:float=0.1423
    ConstraintWeights_2:float=0.0001
    ConstraintWeights_3:float=0.5007
    ConstraintWeights_4:float=0.2453
    ConstraintWeights_5:float=0.0005
    ConstraintWeights_6:float=0.1101
   ) -> (Pose) {
    alphaBoolBlend {blendOption=Linear}
    alphaScaleBiasClamp {scale=0.6}
    boneToModify {boneName=dyn_chain_I_3}
    constraintSetup [
     {targetBone {boneName=upperarm_l}}
     {targetBone {boneName=neck_01}}
     {targetBone {boneName=spine_03}}
     {targetBone {boneName=spine_05}}
     {targetBone {boneName=spine_04}}
     {targetBone {boneName=clavicle_r}}
     {targetBone {boneName=clavicle_l}}
    ]
    constraintWeights [1,1,1,1,1,1,1]
   }
   n63 = "Constraint - dyn_chain_I_4 (PPPPPPPPP)"!(
    ComponentPose=@n62
    Alpha=1
    ConstraintWeights_0=0.0685
    ConstraintWeights_1:float=0.0664
    ConstraintWeights_2:float=0.094
    ConstraintWeights_3:float=0.3726
    ConstraintWeights_4:float=0.0961
    ConstraintWeights_5:float=0.0023
    ConstraintWeights_6:float=0.0509
    ConstraintWeights_7:float=0.233
    ConstraintWeights_8:float=0.0161
   ) -> (Pose) {
    alphaBoolBlend {blendOption=Linear}
    alphaScaleBiasClamp {scale=0.6}
    boneToModify {boneName=dyn_chain_I_4}
    constraintSetup [
     {targetBone {boneName=upperarm_l}}
     {targetBone {boneName=neck_01}}
     {targetBone {boneName=spine_03}}
     {targetBone {boneName=spine_05}}
     {targetBone {boneName=spine_04}}
     {targetBone {boneName=upperarm_r}}
     {targetBone {boneName=clavicle_r}}
     {targetBone {boneName=clavicle_l}}
     {targetBone {boneName=spine_02}}
    ]
    constraintWeights [1,1,1,1,1,1,1,1,1]
   }
   n64 = "Constraint - dyn_chain_I_7 (PPPPPPPP)"!(
    ComponentPose=@n63
    Alpha=1
    ConstraintWeights_0=0.0387
    ConstraintWeights_1:float=0.0513
    ConstraintWeights_2:float=0.4488
    ConstraintWeights_3:float=0.0791
    ConstraintWeights_4:float=0.142
    ConstraintWeights_5:float=0.0226
    ConstraintWeights_6:float=0.0238
    ConstraintWeights_7:float=0.1936
   ) -> (Pose) {
    alphaBoolBlend {blendOption=Linear}
    alphaScaleBiasClamp {scale=0.6}
    boneToModify {boneName=dyn_chain_I_7}
    constraintSetup [
     {targetBone {boneName=neck_01}}
     {targetBone {boneName=spine_04}}
     {targetBone {boneName=clavicle_l}}
     {targetBone {boneName=spine_03}}
     {targetBone {boneName=upperarm_l}}
     {targetBone {boneName=clavicle_r}}
     {targetBone {boneName=spine_02}}
     {targetBone {boneName=spine_05}}
    ]
    constraintWeights [1,1,1,1,1,1,1,1]
   }
   n65 = "Transform (Modify) Bone - Bone: dyn_chain_A_1"(ComponentPose=@n64, Alpha=1, Translation="-0.3,0.4,0", Rotation="-7,0,-2", Scale="1,1,1") {
    rotationMode=BMM_Additive
    translationMode=BMM_Additive
    alphaBoolBlend {blendOption=Linear}
    boneToModify {boneName=dyn_chain_A_1}
    rotation {pitch=-7, roll=-2}
    translation=<-0.3 0.4>
   }
   n66 = "Constraint - dyn_chain_G_1 (PPPPP)"(
    ComponentPose=@n65
    Alpha=1
    ConstraintWeights_0=0.2733
    ConstraintWeights_1=0.3463
    ConstraintWeights_2=0.3404
    ConstraintWeights_3=0.0094
    ConstraintWeights_4=0.0306
   ) {
    alphaBoolBlend {blendOption=Linear}
    boneToModify {boneName=dyn_chain_G_1}
    constraintSetup [
     {targetBone {boneName=neck_02}}
     {targetBone {boneName=head}}
     {targetBone {boneName=neck_01}}
     {targetBone {boneName=spine_04}}
     {targetBone {boneName=spine_05}}
    ]
    constraintWeights [1,1,1,1,1]
   }
   n67 = "Constraint - dyn_chain_G_2 (PPPPPPP)"(
    ComponentPose=@n66
    Alpha=1
    ConstraintWeights_0=0.0842
    ConstraintWeights_1=0.0895
    ConstraintWeights_2=0.015
    ConstraintWeights_3=0.0148
    ConstraintWeights_4=0.5599
    ConstraintWeights_5=0.1712
    ConstraintWeights_6=0.0655
   ) {
    alphaBoolBlend {blendOption=Linear}
    boneToModify {boneName=dyn_chain_G_2}
    constraintSetup [
     {targetBone {boneName=spine_04}}
     {targetBone {boneName=spine_05}}
     {targetBone {boneName=clavicle_l}}
     {targetBone {boneName=clavicle_r}}
     {targetBone {boneName=neck_01}}
     {targetBone {boneName=neck_02}}
     {targetBone {boneName=head}}
    ]
    constraintWeights [1,1,1,1,1,1,1]
   }
   n68 = "Constraint - dyn_chain_G_3 (PPPPPPPP)"(
    ComponentPose=@n67
    Alpha=1
    ConstraintWeights_0=0.0134
    ConstraintWeights_1=0.091
    ConstraintWeights_2=0.0603
    ConstraintWeights_3=0.475
    ConstraintWeights_4=0.1522
    ConstraintWeights_5=0.025
    ConstraintWeights_6=0.0933
    ConstraintWeights_7=0.0898
   ) {
    alphaBoolBlend {blendOption=Linear}
    boneToModify {boneName=dyn_chain_G_3}
    constraintSetup [
     {targetBone {boneName=upperarm_l}}
     {targetBone {boneName=neck_01}}
     {targetBone {boneName=spine_03}}
     {targetBone {boneName=spine_05}}
     {targetBone {boneName=spine_04}}
     {targetBone {boneName=upperarm_r}}
     {targetBone {boneName=clavicle_r}}
     {targetBone {boneName=clavicle_l}}
    ]
    constraintWeights [1,1,1,1,1,1,1,1]
   }
   n69 = "Constraint - dyn_chain_G_5 (PPPPPPPP)"!(
    ComponentPose=@n68
    Alpha=1
    ConstraintWeights_0=0.0024
    ConstraintWeights_1:float=0.0116
    ConstraintWeights_2:float=0.1321
    ConstraintWeights_3:float=0.3684
    ConstraintWeights_4:float=0.188
    ConstraintWeights_5:float=0.2626
    ConstraintWeights_6:float=0.0129
    ConstraintWeights_7:float=0.0221
   ) -> (Pose) {
    alphaBoolBlend {blendOption=Linear}
    boneToModify {boneName=dyn_chain_G_5}
    constraintSetup [
     {targetBone {boneName=pelvis}}
     {targetBone {boneName=spine_01}}
     {targetBone {boneName=spine_02}}
     {targetBone {boneName=spine_03}}
     {targetBone {boneName=spine_04}}
     {targetBone {boneName=spine_05}}
     {targetBone {boneName=clavicle_l}}
     {targetBone {boneName=clavicle_r}}
    ]
    constraintWeights [1,1,1,1,1,1,1,1]
   }
   n70 = "Constraint - dyn_chain_F_2 (PPPPPP)"(
    ComponentPose=@n69
    Alpha=1
    ConstraintWeights_0=0.1298
    ConstraintWeights_1=0.5077
    ConstraintWeights_2=0.2273
    ConstraintWeights_3=0.0004
    ConstraintWeights_4=0.1345
    ConstraintWeights_5=0.0003
   ) {
    alphaBoolBlend {blendOption=Linear}
    boneToModify {boneName=dyn_chain_F_2}
    constraintSetup [
     {targetBone {boneName=neck_01}}
     {targetBone {boneName=spine_05}}
     {targetBone {boneName=spine_04}}
     {targetBone {boneName=upperarm_r}}
     {targetBone {boneName=clavicle_r}}
     {targetBone {boneName=clavicle_l}}
    ]
    constraintWeights [1,1,1,1,1,1]
   }
   n71 = "Constraint - dyn_chain_F_3 (PPPPPPP)"(
    ComponentPose=@n70
    Alpha=1
    ConstraintWeights_0=0.0124
    ConstraintWeights_1=0.1276
    ConstraintWeights_2=0.0708
    ConstraintWeights_3=0.3812
    ConstraintWeights_4=0.0106
    ConstraintWeights_5=0.3038
    ConstraintWeights_6=0.0936
   ) {
    alphaBoolBlend {blendOption=Linear}
    boneToModify {boneName=dyn_chain_F_3}
    constraintSetup [
     {targetBone {boneName=neck_01}}
     {targetBone {boneName=spine_04}}
     {targetBone {boneName=spine_03}}
     {targetBone {boneName=clavicle_r}}
     {targetBone {boneName=spine_02}}
     {targetBone {boneName=spine_05}}
     {targetBone {boneName=upperarm_r}}
    ]
    constraintWeights [1,1,1,1,1,1,1]
   }
   n72 = "Constraint - dyn_chain_F_5 (PPPPPPPP)"(
    ComponentPose=@n71
    Alpha=1
    ConstraintWeights_0=0.308
    ConstraintWeights_1=0.2841
    ConstraintWeights_2=0.2164
    ConstraintWeights_3=0.0147
    ConstraintWeights_4=0.004
    ConstraintWeights_5=0.0082
    ConstraintWeights_6=0.0456
    ConstraintWeights_7=0.119
   ) {
    alphaBoolBlend {blendOption=Linear}
    boneToModify {boneName=dyn_chain_F_5}
    constraintSetup [
     {targetBone {boneName=spine_03}}
     {targetBone {boneName=spine_05}}
     {targetBone {boneName=spine_04}}
     {targetBone {boneName=spine_01}}
     {targetBone {boneName=pelvis}}
     {targetBone {boneName=upperarm_r}}
     {targetBone {boneName=clavicle_r}}
     {targetBone {boneName=spine_02}}
    ]
    constraintWeights [1,1,1,1,1,1,1,1]
   }
   n73 = "Constraint - dyn_chain_F_4 (PPPPPPPP)"(
    ComponentPose=@n72
    Alpha=1
    ConstraintWeights_0=0.2837
    ConstraintWeights_1=0.2934
    ConstraintWeights_2=0.2229
    ConstraintWeights_3=0.0101
    ConstraintWeights_4=0.0027
    ConstraintWeights_5=0.0157
    ConstraintWeights_6=0.0865
    ConstraintWeights_7=0.0851
   ) {
    alphaBoolBlend {blendOption=Linear}
    boneToModify {boneName=dyn_chain_F_4}
    constraintSetup [
     {targetBone {boneName=spine_03}}
     {targetBone {boneName=spine_05}}
     {targetBone {boneName=spine_04}}
     {targetBone {boneName=spine_01}}
     {targetBone {boneName=pelvis}}
     {targetBone {boneName=upperarm_r}}
     {targetBone {boneName=clavicle_r}}
     {targetBone {boneName=spine_02}}
    ]
    constraintWeights [1,1,1,1,1,1,1,1]
   }
   n74 = "Constraint - dyn_chain_I_2 (PPPPPP)"(
    ComponentPose=@n73
    Alpha=1
    ConstraintWeights_0=0.0121
    ConstraintWeights_1=0.0423
    ConstraintWeights_2=0.0044
    ConstraintWeights_3=0.3612
    ConstraintWeights_4=0.321
    ConstraintWeights_5=0.259
   ) {
    alphaBoolBlend {blendOption=Linear}
    alphaScaleBiasClamp {scale=0.7}
    boneToModify {boneName=dyn_chain_I_2}
    constraintSetup [
     {targetBone {boneName=spine_04}}
     {targetBone {boneName=spine_05}}
     {targetBone {boneName=clavicle_l}}
     {targetBone {boneName=neck_01}}
     {targetBone {boneName=neck_02}}
     {targetBone {boneName=head}}
    ]
    constraintWeights [1,1,1,1,1,1]
   }
   n75 = "Constraint - dyn_chain_I_3 (PPPPPPP)"(
    ComponentPose=@n74
    Alpha=1
    ConstraintWeights_0=0.001
    ConstraintWeights_1=0.1423
    ConstraintWeights_2=0.0001
    ConstraintWeights_3=0.5007
    ConstraintWeights_4=0.2453
    ConstraintWeights_5=0.0005
    ConstraintWeights_6=0.1101
   ) {
    alphaBoolBlend {blendOption=Linear}
    boneToModify {boneName=dyn_chain_I_3}
    constraintSetup [
     {targetBone {boneName=upperarm_l}}
     {targetBone {boneName=neck_01}}
     {targetBone {boneName=spine_03}}
     {targetBone {boneName=spine_05}}
     {targetBone {boneName=spine_04}}
     {targetBone {boneName=clavicle_r}}
     {targetBone {boneName=clavicle_l}}
    ]
    constraintWeights [1,1,1,1,1,1,1]
   }
   n76 = "Constraint - dyn_chain_I_4 (PPPPPPPPP)"(
    ComponentPose=@n75
    Alpha=1
    ConstraintWeights_0=0.0685
    ConstraintWeights_1=0.0664
    ConstraintWeights_2=0.094
    ConstraintWeights_3=0.3726
    ConstraintWeights_4=0.0961
    ConstraintWeights_5=0.0023
    ConstraintWeights_6=0.0509
    ConstraintWeights_7=0.233
    ConstraintWeights_8=0.0161
   ) {
    alphaBoolBlend {blendOption=Linear}
    boneToModify {boneName=dyn_chain_I_4}
    constraintSetup [
     {targetBone {boneName=upperarm_l}}
     {targetBone {boneName=neck_01}}
     {targetBone {boneName=spine_03}}
     {targetBone {boneName=spine_05}}
     {targetBone {boneName=spine_04}}
     {targetBone {boneName=upperarm_r}}
     {targetBone {boneName=clavicle_r}}
     {targetBone {boneName=clavicle_l}}
     {targetBone {boneName=spine_02}}
    ]
    constraintWeights [1,1,1,1,1,1,1,1,1]
   }
   n77 = "Constraint - dyn_chain_I_7 (PPPPPPPP)"(
    ComponentPose=@n76
    Alpha=1
    ConstraintWeights_0=0.0387
    ConstraintWeights_1=0.0513
    ConstraintWeights_2=0.4488
    ConstraintWeights_3=0.0791
    ConstraintWeights_4=0.142
    ConstraintWeights_5=0.0226
    ConstraintWeights_6=0.0238
    ConstraintWeights_7=0.1936
   ) {
    alphaBoolBlend {blendOption=Linear}
    boneToModify {boneName=dyn_chain_I_7}
    constraintSetup [
     {targetBone {boneName=neck_01}}
     {targetBone {boneName=spine_04}}
     {targetBone {boneName=clavicle_l}}
     {targetBone {boneName=spine_03}}
     {targetBone {boneName=upperarm_l}}
     {targetBone {boneName=clavicle_r}}
     {targetBone {boneName=spine_02}}
     {targetBone {boneName=spine_05}}
    ]
    constraintWeights [1,1,1,1,1,1,1,1]
   }
   n78 = "Constraint - dyn_chain_F_8 (P)"(ComponentPose=@n77, Alpha=1, ConstraintWeights_0=1) {
    alphaBoolBlend {blendOption=Linear}
    alphaScaleBiasClamp {scale=0.7}
    boneToModify {boneName=dyn_chain_F_8}
    constraintSetup [{targetBone {boneName=dyn_chain_F_2}}]
    constraintWeights [1]
   }
   n79 = "Transform (Modify) Bone - Bone: dyn_chain_G_2"!(ComponentPose=@n78, Alpha=1, Scale="0.7,0.7,0.7") -> (Pose) {
    scaleMode=BMM_Additive
    scaleSpace=BCS_BoneSpace
    alphaBoolBlend {blendOption=Linear}
    boneToModify {boneName=dyn_chain_G_2}
    scale=<0.7 0.7 0.7>
   }
   n80 = "Transform (Modify) Bone - Bone: dyn_chain_F_8"!(ComponentPose=@n79, Alpha=1, Scale="0.7,0.7,0.7") -> (Pose) {
    scaleMode=BMM_Additive
    scaleSpace=BCS_BoneSpace
    alphaBoolBlend {blendOption=Linear}
    boneToModify {boneName=dyn_chain_F_8}
    scale=<0.7 0.7 0.7>
   }
   n81 = "Transform (Modify) Bone - Bone: dyn_chain_H_2"!(ComponentPose=@n80, Alpha=1, Scale="0.8,0.8,0.8") -> (Pose) {
    scaleMode=BMM_Additive
    scaleSpace=BCS_BoneSpace
    alphaBoolBlend {blendOption=Linear}
    boneToModify {boneName=dyn_chain_H_2}
    scale=<0.8 0.8 0.8>
   }
   n82 = "Transform (Modify) Bone - Bone: dyn_chain_I_2"!(ComponentPose=@n81, Alpha=1, Scale="0.8,0.8,0.8") -> (Pose) {
    scaleMode=BMM_Additive
    scaleSpace=BCS_BoneSpace
    alphaBoolBlend {blendOption=Linear}
    boneToModify {boneName=dyn_chain_I_2}
    scale=<0.8 0.8 0.8>
   }
   Output Pose(Result=Component To Local(ComponentPose=@n82)) {layerGroup=None}
  ]}
 EventGraph {nodes [
   Event BlueprintUpdateAnimation()
   > n1 = Sequence!(execute) -> (then_0, then_1, then_2:Exec)
   > n2 = Gravity Override State Function DEPRECATED(CharacterPartAnimInstance=Self-Reference(), GameTimeInSeconds=GetGameTimeInSeconds(), GravityOverrideParams=Get Gravity Override Params())
   > n3 = Set GravityOverride(GravityOverride=@n2)
   > Set GravityOverride(GravityOverride=SelectVector(A="0,0,-2000", B=@n3, bPickA=Get isWearingBackblingBool() -> (isWearingBackblingBool:bool)))
   n4 = GetPartSkeletalMeshComponent()
   n5 = BreakRotator(InRot=BreakTransform(InTransform=GetDeltaTransformFromRefPose(self=@n4, BoneName=C_jaw, BaseName=head)).Rotation)
   n6 = BreakRotator(InRot=BreakTransform(InTransform=GetDeltaTransformFromRefPose(self=@n4, BoneName=neck_02, BaseName=spine_05)).Rotation)
   > Set jawOpenLookingDown(jawOpenLookingDown:double="Max (Float)"(A=MapRangeClamped(Value=@n5, InRangeA=5, InRangeB=20), B=MapRangeClamped(Value=@n6.Yaw, InRangeA=-15, InRangeB=7))) -> (Output_Get:double)
   > Set neckBack(neckBack:double=MapRangeClamped(Value=@n6.Yaw, InRangeA=-20, InRangeB=-40)) -> (Output_Get:double)
   n7 = Branch(execute=@n1.then_2, Condition=Get bIsDBNO())
   Do Once(execute=@n7.else, Reset=@n7.then, Start Closed=true)
   > Delay(Duration=0.2)
   > ResetDynamics(InTeleportType=TeleportPhysics)
   > ResetDynamics(InTeleportType=ResetPhysics)
   n8 = Branch(execute=@n1.then_1, Condition=Get bIsInFrontEnd())
   > Set isPlayingFrontendIdle(isPlayingFrontendIdle:float=GetCurveValue(self=GetOwnerFrontEndAnimInstance(), CurveName=Frontend_Idle_Bypass)) -> (Output_Get:float)
   > Set isNOT_PlayingFrontendIdle(isNOT_PlayingFrontendIdle:float=0) -> (Output_Get:float)
   Set isNOT_PlayingFrontendIdle(execute=@n8.else, isNOT_PlayingFrontendIdle=1)
   > Set isPlayingFrontendIdle(isPlayingFrontendIdle=0)
   Event OnCharacterCustomizationLoaded()
   > Set isWearingBackblingBool(isWearingBackblingBool:bool=CheckForBackbling()) -> (Output_Get:bool)
   n9 = Make Array("[0]"=(TagName="Cosmetics.Variant.Property.Stage2"))
   > Set hatOff(hatOff:bool=NOT Boolean(A=CharacterItemDefinitionHasAnyMetaTag(Tags=@n9))) -> (Output_Get:bool)
  ]}
]
variables [
 Gravity Override Params {type=GravityOverrideParamsStruct, default {skydive_FloatingDown {gravityOverride=<0 -700 -200>, windAmplitude=<50 -100 100>}}}
 GravityOverride {type=Vector}
 hatOff {type=boolean}
 isNOT_PlayingFrontendIdle {type=number}
 isPlayingFrontendIdle {type=number}
 isWearingBackblingBool {type=boolean}
 jawOpenLookingDown {type=number}
 neckBack {type=number}
]
